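# M_MatteBlack is an unlit, two-sided material in the cinematic animation folder that, going by its name, provides a flat matte black surface.
# The graph has no expressions, so emissive stays at its default black and the surface ignores scene lighting.
# Automatic setting of usage flags in the editor is turned off.
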Material /ToyFoxGameplay/Assets/Animations/Cinematic/M_MatteBlack
graph {expressions {}, properties {bAutomaticallySetUsageInEditor=false, shadingModel=MSM_Unlit, twoSided=true, usedShadingModels=Unlit}}
